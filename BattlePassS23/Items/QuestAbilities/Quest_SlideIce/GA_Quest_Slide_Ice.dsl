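; Battle Pass S23 quest ability that tracks how far the player slides on ice.
; An event-activated ability starts a looping timer at a random sub-second interval, checking sliding and icy-feet tags.
; Each tick adds floored horizontal meters; leaving ice resets the total, and not sliding also ends the ability.
; Reaching MinDistance sends the SlideIce stat event to the quest manager and ends the ability, which clears the timer.
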(event Ability|EventActivateAbilityFromEvent
  (params (EventData :type Gameplay%20Event%20Data%20Structure%20%28by%20ref%29))
  ;[1 1 1] Activate ability
  (|CallParentFunction EventData)
  (bind _output_get (Variables|Default|SetPreviousLocation (Transformation|GetActorLocation (Variables|GAVars|Data|GetOwningPlayerPawn))))
  (Variables|Default|SetCurrentLocation _output_get)
  (bind _output_get_2 (Variables|Default|SetInterval (Math|Random|RandomFloatinRange 0.24 0.5)))
  (bind _returnvalue (Utilities|Time|SetTimerbyEvent (delegate Custom|SendDistanceUpdate) _output_get_2 true))
  (Variables|Default|SetTImerHandle _returnvalue))

(event Custom|SendDistanceUpdate
  (bind _owningplayerpawn (Variables|GAVars|Data|GetOwningPlayerPawn))
  (bind (_previouslocation_x _previouslocation_y) (Math|Vector|BreakVector (Variables|GetPreviousLocation_X)))
  (bind (_currentlocation_x _currentlocation_y) (Math|Vector|BreakVector (Variables|GetCurrentLocation_X)))
  ;[1 1 1] Check for Ice and Sliding and add up distance  then fire complex when met.
  (if (GameplayTags|HasMatchingGameplayTag _owningplayerpawn "(TagName=\"MovementMode.Sliding\")")
    (if (GameplayTags|HasMatchingGameplayTag _owningplayerpawn "(TagName=\"Gameplay.Status.IcyFeet.Surface\")")
      (Variables|Default|SetCurrentLocation (Transformation|GetActorLocation _owningplayerpawn))
      ;;[1 1 1] Continuous adding of distance while sliding on ice (horizontal only)
      (Variables|Default|SetMetersTraveled (+ (Math|Float|Floor (/ (Math|Vector|VectorLength (Math|Vector|vector-vector (Math|Vector|MakeVector _currentlocation_x _currentlocation_y) (Math|Vector|MakeVector _previouslocation_x _previouslocation_y))) 1e+02)) (Variables|Default|GetMetersTraveled)))
      (Variables|Default|SetPreviousLocation (Variables|Default|GetCurrentLocation)) ; Reset location
      ;;[1 1 1] If min distance is met, fire the complex
      (if (>= (Variables|Default|GetMetersTraveled) (Variables|Default|GetMinDistance))
        (Quest|SendComplexCustomStatEvent (Variables|GAVars|Data|GetOwningPlayerQuestManager) 0 "(GameplayTags=((TagName=\"Athena.Quests.S23.SlideIce\")))" "(GameplayTags=)" 1)
        (Ability|EndAbility))
      (else
        (Variables|Default|SetMetersTraveled 0)))
    (else
      (Variables|Default|SetMetersTraveled 0)
      (Ability|EndAbility))))

(event Ability|EventOnEndAbility
  (params (bWasCancelled :type Boolean))
  (bind _timerhandle (Variables|Default|GetTImerHandle))
  ;[1 1 1] On End Ability
  (|CallParentFunction bWasCancelled)
  (if (Utilities|Time|IsValidTimerHandle _timerhandle)
    (Utilities|Time|ClearandInvalidateTimerbyHandle _timerhandle)))
